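# SoundCue for the SnowfallFemale cosmetic glider's open sound.
# A mixer plays a fixed open sound layered with one of three equally weighted random dragon variations.
# It uses the shared player glider open/close attenuation and concurrency, a glider-open audio property sheet, and the glider landing sound class.
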SoundCue /BRCosmetics/Sounds/Gliders/SnowfallFemale/GliderOpenVariations_SnowfallFemale
root=Mixer_1
nodes [
 {id=Mixer_1, children [Random_0,WavePlayer_0], properties {inputVolume [1,1]}}
 {id=Random_0, children [WavePlayer_1,WavePlayer_2,WavePlayer_3], properties {weights [1,1,1]}}
 {id=WavePlayer_0, properties {soundWaveAssetPtr=&/Glider_SnowfallFemale_Open}}
 {id=WavePlayer_1, properties {soundWaveAssetPtr=&/Glider_SnowfallFemale_Open_Dragon_01}}
 {id=WavePlayer_2, properties {soundWaveAssetPtr=&/Glider_SnowfallFemale_Open_Dragon_02}}
 {id=WavePlayer_3, properties {soundWaveAssetPtr=&/Glider_SnowfallFemale_Open_Dragon_03}}
]
properties {
 attenuationSettings=/Game/Sounds/Attenuation/Fort_Player_Glider_OpenClose_Athena_Att
 audioPropertiesSheet=/BRCosmetics/Sounds/Settings/PropertySheets/Glider/APS_Cosmetic_Glider_Open
 maxDistance=10000
 soundClassObject=&Fort_Gliders_Land
 volumeMultiplier=0.8
 concurrencySet [&Athena_GliderOpenClose_Conc]
}
