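# Sound cue for the looping thrust sound of the SnowfallFemale glider cosmetic.
# It mixes two looping layers: a wind loop and a magic loop that plays unmodulated.
# A continuous modulator drives the wind layer's pitch and volume from the AnyDirection parameter.
# It uses the glider thrust attenuation, cosmetic glider loop property sheet and glider sound class.
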SoundCue /BRCosmetics/Sounds/Gliders/SnowfallFemale/GliderThrustLoop_SnowfallFemale1
root=Mixer_0
nodes [
 {id=Mixer_0, children [ModulatorContinuous_0, WavePlayer_2], properties {inputVolume [1,1]}}
 {
  id=ModulatorContinuous_0
  children [WavePlayer_1]
  properties {
   pitchModulationParams {default=0, maxOutput=1.4, minOutput=1, parameterName=AnyDirection}
   volumeModulationParams {default=0, maxOutput=1.3, minOutput=0.8583, parameterName=AnyDirection}
  }
 }
 {
  id=WavePlayer_1
  properties {bLooping=true, soundWaveAssetPtr=&/Glider_SnowfallFemale_ThrustLoop_Wind}
 }
 {
  id=WavePlayer_2
  properties {bLooping=true, soundWaveAssetPtr=&/Glider_SnowfallFemale_ThrustLoop_Magic}
 }
]
properties {
 attenuationSettings=&Fort_Player_Glider_Thrust_Athena_Att
 audioPropertiesSheet=&APS_Cosmetic_Glider_Loop
 maxDistance=12500
 soundClassObject=&Fort_Gliders_Land
 volumeMultiplier=1.2
 concurrencySet [null]
}
